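# Physics asset for the medium male BriteDino player body mesh, covering secondary motion of its armband knot and three scarf pieces (front, low, back).
# Each piece pairs a small capsule on a base bone with an offset capsule on a dyn bone, linked by its own constraint.
# Per-piece angular limits keep the armband knot nearly rigid and give the front scarf the widest swing and twist.
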PhysicsAsset /Game/Characters/Player/Male/Medium/Bodies/M_MED_BriteDino/Meshes/M_MED_BriteDino_Physics
bodies {
 base_armband_knot [{shapeType=Capsule, length=2, radius=2, shapeName=base_armband_knot_capsule, rotation=<rot -90 0 0>}]
 base_scarf_back [{shapeType=Capsule, length=1, radius=1, shapeName=base_scarf_low_capsule, rotation=<rot 90 0 0>}]
 base_scarf_f [{shapeType=Capsule, length=1, radius=1, shapeName=base_scarf_f_capsule, rotation=<rot 90 0 0>}]
 base_scarf_low [{shapeType=Capsule, length=1, radius=1, shapeName=base_scarf_low_capsule, rotation=<rot 90 0 0>}]
 dyn_armband_knot [{shapeType=Capsule, length=1.5, radius=4, shapeName=dyn_armband_knot_capsule, center=<3 0 0>, rotation=<rot -90 0 0>}]
 dyn_scarf_back [{shapeType=Capsule, length=1.5, radius=1.5, shapeName=dyn_scarf_low_capsule, center=<1 0 0>, rotation=<rot 90 0 0>}]
 dyn_scarf_f [{shapeType=Capsule, length=1.5, radius=1.5, shapeName=dyn_scarf_f_capsule, center=<1 0 0>, rotation=<rot 90 0 0>}]
 dyn_scarf_low [{shapeType=Capsule, length=1.5, radius=1.5, shapeName=dyn_scarf_low_capsule, center=<1 0 0>, rotation=<rot 90 0 0>}]
}
constraints [
 {bone1Name=dyn_armband_knot, bone2Name=base_armband_knot, swing1LimitDegrees=5, swing2LimitDegrees=5, twistLimitDegrees=1}
 {bone1Name=dyn_scarf_low, bone2Name=base_scarf_low, swing1LimitDegrees=35, swing2LimitDegrees=30, twistLimitDegrees=5}
 {bone1Name=dyn_scarf_f, bone2Name=base_scarf_f, swing1LimitDegrees=45, swing2LimitDegrees=45, twistLimitDegrees=25}
 {bone1Name=dyn_scarf_back, bone2Name=base_scarf_back, swing1LimitDegrees=25, swing2LimitDegrees=20, twistLimitDegrees=0}
]
