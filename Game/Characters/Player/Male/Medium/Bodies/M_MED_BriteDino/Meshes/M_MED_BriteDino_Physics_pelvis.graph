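# Physics asset for the pelvis region of the BriteDino medium male player body, set up for six hanging flaps, three per side.
# Each dyn_flap capsule is constrained to a tiny base_flap anchor with limited swing and near-zero or locked twist, with some tighter limits on the right.
# Base flaps and thighs link to the pelvis with free rotation, while large pelvis and thigh capsules provide body volumes around the flaps.
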PhysicsAsset /Game/Characters/Player/Male/Medium/Bodies/M_MED_BriteDino/Meshes/M_MED_BriteDino_Physics_pelvis
bodies {
 base_flap_01_l [{shapeType=Capsule, length=1, radius=1, shapeName=base_flap_01_l_capsule, rotation=<rot 90 0 0>}]
 base_flap_01_r [{shapeType=Capsule, length=1, radius=1, shapeName=base_flap_01_r_capsule, rotation=<rot 90 0 0>}]
 base_flap_02_l [{shapeType=Capsule, length=1, radius=1, shapeName=base_flap_02_l_capsule, rotation=<rot 90 0 0>}]
 base_flap_02_r [{shapeType=Capsule, length=1, radius=1, shapeName=base_flap_02_r_capsule, rotation=<rot 90 0 0>}]
 base_flap_03_l [{shapeType=Capsule, length=1, radius=1, shapeName=base_flap_03_l_capsule, rotation=<rot 90 0 0>}]
 base_flap_03_r [{shapeType=Capsule, length=1, radius=1, shapeName=base_flap_03_r_capsule, rotation=<rot 90 0 0>}]
 dyn_flap_01_l [{shapeType=Capsule, length=5, radius=2, shapeName=dyn_flap_01_l_capsule, center=<5 0 0.4304>, rotation=<rot 90 0 0>}]
 dyn_flap_01_r [{shapeType=Capsule, length=5, radius=2, shapeName=dyn_flap_01_r_capsule, center=<3 0 2.5>, rotation=<rot 90 0 0>}]
 dyn_flap_02_l [{shapeType=Capsule, length=5, radius=2, shapeName=dyn_flap_02_l_capsule, center=<5 0 0>, rotation=<rot 90 0 0>}]
 dyn_flap_02_r [{shapeType=Capsule, length=5, radius=2, shapeName=dyn_flap_02_r_capsule, center=<5 0 0>, rotation=<rot 90 0 0>}]
 dyn_flap_03_l [{shapeType=Capsule, length=5, radius=2, shapeName=dyn_flap_03_l_capsule, center=<5 0 0.4507>, rotation=<rot 90 0 0>}]
 dyn_flap_03_r [{shapeType=Capsule, length=5, radius=2, shapeName=dyn_flap_03_r_capsule, center=<5 0 0>, rotation=<rot 90 0 0>}]
 pelvis [{shapeType=Capsule, length=12, radius=9.5755, center=<0 -1.4954 0>}]
 thigh_l [{shapeType=Capsule, length=36, radius=9, center=<-22 1.7974 2.1547>, rotation=<rot 0 -90 -90>}]
 thigh_r [{shapeType=Capsule, length=38.5, radius=9, center=<22 0 -0.5>, rotation=<rot 0 -90 -90>}]
}
constraints [
 {bone1Name=dyn_flap_01_l, bone2Name=base_flap_01_l, swing1LimitDegrees=45, swing2LimitDegrees=45, twistLimitDegrees=1}
 {bone1Name=dyn_flap_01_r, bone2Name=base_flap_01_r, swing1LimitDegrees=35, swing2LimitDegrees=25, twistLimitDegrees=0, twistMotion=Locked}
 {bone1Name=dyn_flap_02_l, bone2Name=base_flap_02_l, swing1LimitDegrees=45, swing2LimitDegrees=45, twistLimitDegrees=1}
 {bone1Name=dyn_flap_02_r, bone2Name=base_flap_02_r, swing1LimitDegrees=10, swing2LimitDegrees=25, twistLimitDegrees=0}
 {bone1Name=dyn_flap_03_l, bone2Name=base_flap_03_l, swing1LimitDegrees=45, swing2LimitDegrees=25, twistLimitDegrees=1}
 {bone1Name=dyn_flap_03_r, bone2Name=base_flap_03_r, swing1LimitDegrees=45, swing2LimitDegrees=25, twistLimitDegrees=1}
 {
  bone1Name=base_flap_02_l
  bone2Name=pelvis
  swing1LimitDegrees=45
  swing1Motion=Free
  swing2LimitDegrees=45
  swing2Motion=Free
  twistLimitDegrees=45
  twistMotion=Free
 }
 {
  bone1Name=base_flap_03_l
  bone2Name=pelvis
  swing1LimitDegrees=45
  swing1Motion=Free
  swing2LimitDegrees=45
  swing2Motion=Free
  twistLimitDegrees=45
  twistMotion=Free
 }
 {
  bone1Name=base_flap_02_r
  bone2Name=pelvis
  swing1LimitDegrees=45
  swing1Motion=Free
  swing2LimitDegrees=45
  swing2Motion=Free
  twistLimitDegrees=45
  twistMotion=Free
 }
 {
  bone1Name=base_flap_03_r
  bone2Name=pelvis
  swing1LimitDegrees=45
  swing1Motion=Free
  swing2LimitDegrees=45
  swing2Motion=Free
  twistLimitDegrees=45
  twistMotion=Free
 }
 {
  bone1Name=base_flap_01_l
  bone2Name=pelvis
  swing1LimitDegrees=45
  swing1Motion=Free
  swing2LimitDegrees=45
  swing2Motion=Free
  twistLimitDegrees=45
  twistMotion=Free
 }
 {
  bone1Name=base_flap_01_r
  bone2Name=pelvis
  swing1LimitDegrees=45
  swing1Motion=Free
  swing2LimitDegrees=45
  swing2Motion=Free
  twistLimitDegrees=45
  twistMotion=Free
 }
 {
  bone1Name=thigh_r
  bone2Name=pelvis
  swing1LimitDegrees=45
  swing1Motion=Free
  swing2LimitDegrees=45
  swing2Motion=Free
  twistLimitDegrees=45
  twistMotion=Free
 }
 {
  bone1Name=thigh_l
  bone2Name=pelvis
  swing1LimitDegrees=45
  swing1Motion=Free
  swing2LimitDegrees=45
  swing2Motion=Free
  twistLimitDegrees=45
  twistMotion=Free
 }
]
