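# Post-processes the BriteDino player body's pose, copied from the parent mesh, adding secondary motion and corrective bone offsets.
# A Control Rig and hip flap offsets feed two RigidBody sims and AnimDynamics bands, cached before additive arm and leg-flap correctives.
# Each update derives gravity override and corrective alphas from limb rotation deltas and parachute state.
# Front-end or DBNO state switches leg-flap rotations.
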AnimBlueprint /Game/Characters/Player/Male/Medium/Bodies/M_MED_BriteDino/Meshes/M_MED_BriteDino_AnimBP
target_skeleton=&/M_MED_BriteDino_Skeleton
graphs [
 AnimGraph {nodes [
   n1 = "Transform (Modify) Bone - Bone: deform_upperarm_up_r"(
    ComponentPose=Local To Component(LocalPose=Use cached pose 'RB_dynamics_cache'())
    Alpha=Get UpperArm_Rot_Alpha() -> (UpperArm_Rot_Alpha:double)
    Translation="5,0,2.5"
    Rotation="0,10,0"
    Scale="1,1,1"
   ) {
    rotationMode=BMM_Additive
    rotationSpace=BCS_BoneSpace
    translationMode=BMM_Additive
    translationSpace=BCS_BoneSpace
    alphaBoolBlend {blendOption=Linear}
    alphaScaleBiasClamp {interpSpeedDecreasing=15}
    boneToModify {boneName=deform_upperarm_up_r}
    rotation {yaw=10}
    translation {x=5, z=2.5}
   }
   n2 = Get rotationBigFlap() -> (rotationBigFlap:Rotator)
   n3 = Get calf_l() -> (calf_l:double)
   n4 = "Transform (Modify) Bone - Bone: base_flap_leg_big_l"(ComponentPose=@n1, Alpha=@n3, Translation="0,0,0", Rotation=@n2, Scale="0.75,1,1") {
    rotationMode=BMM_Additive
    rotationSpace=BCS_BoneSpace
    scaleMode=BMM_Additive
    scaleSpace=BCS_BoneSpace
    translationMode=BMM_Additive
    translationSpace=BCS_BoneSpace
    alphaBoolBlend {blendOption=Linear}
    alphaScaleBiasClamp {interpSpeedDecreasing=30, interpSpeedIncreasing=0}
    boneToModify {boneName=base_flap_leg_big_l}
    rotation {pitch=-80}
    scale {x=0.75}
   }
   n5 = Get rotationLittleFlap() -> (rotationLittleFlap:Rotator)
   n6 = "Transform (Modify) Bone - Bone: base_flap_leg_little_l"(ComponentPose=@n4, Alpha=@n3, Translation="1,0,2.5", Rotation=@n5, Scale="1,1,1") {
    rotationMode=BMM_Additive
    rotationSpace=BCS_BoneSpace
    scaleMode=BMM_Additive
    scaleSpace=BCS_BoneSpace
    translationMode=BMM_Additive
    translationSpace=BCS_BoneSpace
    alphaBoolBlend {blendOption=Linear}
    alphaScaleBiasClamp {interpSpeedDecreasing=30, interpSpeedIncreasing=0}
    boneToModify {boneName=base_flap_leg_little_l}
    rotation {pitch=-45}
    translation {x=1, z=2.5}
   }
   n7 = Get calf_r() -> (calf_r:double)
   n8 = "Transform (Modify) Bone - Bone: base_flap_leg_big_r"(ComponentPose=@n6, Alpha=@n7, Translation="0,0,0", Rotation=@n2, Scale="0.75,1,1") {
    rotationMode=BMM_Additive
    rotationSpace=BCS_BoneSpace
    scaleMode=BMM_Additive
    scaleSpace=BCS_BoneSpace
    translationMode=BMM_Additive
    translationSpace=BCS_BoneSpace
    alphaBoolBlend {blendOption=Linear}
    alphaScaleBiasClamp {interpSpeedDecreasing=30, interpSpeedIncreasing=0}
    boneToModify {boneName=base_flap_leg_big_r}
    scale {x=0.75}
   }
   n9 = "Transform (Modify) Bone - Bone: base_flap_leg_little_r"(ComponentPose=@n8, Alpha=@n7, Translation="1,0,2.5", Rotation=@n5, Scale="1,1,1") {
    rotationMode=BMM_Additive
    rotationSpace=BCS_BoneSpace
    scaleMode=BMM_Additive
    scaleSpace=BCS_BoneSpace
    translationMode=BMM_Additive
    translationSpace=BCS_BoneSpace
    alphaBoolBlend {blendOption=Linear}
    alphaScaleBiasClamp {interpSpeedDecreasing=30, interpSpeedIncreasing=0}
    boneToModify {boneName=base_flap_leg_little_r}
    translation {x=1, z=2.5}
   }
   n10 = Get skydiving() -> (skydiving:double)
   n11 = "Transform (Modify) Bone - Bone: deform_upperarm_up_l"(ComponentPose=@n9, Alpha=@n10, Translation="2.5,0,-2.5", Rotation="0,0,0", Scale="1,1,0.5") {
    rotationMode=BMM_Additive
    rotationSpace=BCS_BoneSpace
    scaleMode=BMM_Additive
    scaleSpace=BCS_BoneSpace
    translationMode=BMM_Additive
    translationSpace=BCS_BoneSpace
    alphaBoolBlend {blendOption=Linear}
    alphaScaleBiasClamp {bInterpResult=true}
    boneToModify {boneName=deform_upperarm_up_l}
    scale {z=0.5}
    translation {x=2.5, z=-2.5}
   }
   n12 = "Transform (Modify) Bone - Bone: deform_upperarm_up_r"(ComponentPose=@n11, Alpha=@n10, Translation="-5,0,3", Rotation="0,0,0", Scale="1,1,0.75") {
    rotationMode=BMM_Additive
    rotationSpace=BCS_BoneSpace
    scaleMode=BMM_Additive
    scaleSpace=BCS_BoneSpace
    translationMode=BMM_Additive
    translationSpace=BCS_BoneSpace
    alphaBoolBlend {blendOption=Linear}
    alphaScaleBiasClamp {bInterpResult=true}
    boneToModify {boneName=deform_upperarm_up_r}
    scale {z=0.75}
    translation {x=-5, z=3}
   }
   Output Pose(Result=Component To Local(ComponentPose=@n12)) {layerGroup=None}
   n13 = Control Rig(Source=Copy Pose From Mesh() {bUseAttachedParent=true}) {
    alphaBoolBlend {blendOption=Linear}
    controlRigAssetReference {blueprintRigClass=&/M_MED_BriteDino_ControlRig.M_MED_BriteDino_ControlRig_C}
   }
   n14 = Get legRotation_L() -> (legRotation_L:double)
   n15 = "Transform (Modify) Bone - Bone: base_flap_01_l"(ComponentPose=Local To Component(LocalPose=@n13), Alpha=@n14, Translation="-12,0,0", Rotation="60,0,30", Scale="1,1,1") {
    rotationMode=BMM_Additive
    rotationSpace=BCS_BoneSpace
    translationMode=BMM_Additive
    translationSpace=BCS_BoneSpace
    alphaBoolBlend {blendOption=Linear}
    boneToModify {boneName=base_flap_01_l}
    rotation {pitch=60, roll=30}
    translation {x=-12}
   }
   n16 = "Transform (Modify) Bone - Bone: base_flap_02_l"(ComponentPose=@n15, Alpha=@n14, Translation="-8,0,0", Rotation="0,0,-75", Scale="1,1,1") {
    rotationMode=BMM_Additive
    rotationSpace=BCS_BoneSpace
    translationMode=BMM_Additive
    translationSpace=BCS_BoneSpace
    alphaBoolBlend {blendOption=Linear}
    boneToModify {boneName=base_flap_02_l}
    rotation {roll=-75}
    translation {x=-8}
   }
   n17 = Get legRotation_R() -> (legRotation_R:double)
   n18 = "Transform (Modify) Bone - Bone: base_flap_01_r"(ComponentPose=@n16, Alpha=@n17, Translation="-3,0,1", Rotation="0,-25,0", Scale="1,1,1") {
    rotationMode=BMM_Additive
    rotationSpace=BCS_BoneSpace
    scaleMode=BMM_Additive
    scaleSpace=BCS_BoneSpace
    translationMode=BMM_Additive
    translationSpace=BCS_BoneSpace
    alphaBoolBlend {blendOption=Linear}
    boneToModify {boneName=base_flap_01_r}
    rotation {yaw=-25}
    translation {x=-3, z=1}
   }
   n19 = "Transform (Modify) Bone - Bone: base_flap_02_r"(ComponentPose=@n18, Alpha=@n17, Translation="-1,0,1", Rotation="0,-10,0", Scale="1,1,1") {
    rotationMode=BMM_Additive
    rotationSpace=BCS_BoneSpace
    scaleMode=BMM_Additive
    scaleSpace=BCS_BoneSpace
    translationMode=BMM_Additive
    translationSpace=BCS_BoneSpace
    alphaBoolBlend {blendOption=Linear}
    boneToModify {boneName=base_flap_02_r}
    rotation {yaw=-10}
    translation {x=-1, z=1}
   }
   n20 = RigidBody!(ComponentPose=@n19, Alpha, OverrideWorldGravity:Vector=Get GravityOverride(), ExternalForce) -> (Pose) {
    bClampLinearTranslationLimitToRefPose=true
    bOverrideWorldGravity=true
    overridePhysicsAsset=&/M_MED_BriteDino_Physics
    simulationSpace=BaseBoneSpace
    alphaBoolBlend {blendOption=Linear}
    baseBoneRef {boneName=spine_05}
    componentAppliedLinearAccClamp=<3000 3000 3000>
    componentLinearAccScale=<0.3 0.4 0.4>
    componentLinearVelScale=<0.3 0.4 0.4>
    simSpaceSettings {
     maxAngularAcceleration=1000
     maxAngularVelocity=3
     maxLinearAcceleration=700
     maxLinearVelocity=500
     velocityScaleZ=0
     worldAlpha=0.2375
     externalLinearDragV=<0.2 0.2>
    }
   }
   n21 = RigidBody(ComponentPose=@n20, OverrideWorldGravity=Get GravityOverride()) {
    bClampLinearTranslationLimitToRefPose=true
    bOverrideWorldGravity=true
    overridePhysicsAsset=&/M_MED_BriteDino_Physics_pelvis
    simulationSpace=BaseBoneSpace
    alphaBoolBlend {blendOption=Linear}
    baseBoneRef {boneName=pelvis}
    componentAppliedLinearAccClamp=<3000 3000 3000>
    componentLinearAccScale=<0.3 0.4 0.4>
    componentLinearVelScale=<0.3 0.4 0.4>
    simSpaceSettings {
     maxAngularAcceleration=1000
     maxAngularVelocity=3
     maxLinearAcceleration=700
     maxLinearVelocity=500
     velocityScaleZ=0
     worldAlpha=0.2375
     externalLinearDragV=<0.2 0.2>
    }
   }
   n22 = "Transform (Modify) Bone - Bone: base_flap_01_r"(ComponentPose=@n21, Alpha=0, Translation="-1,0,0", Rotation="0,0,0", Scale="1,1,1") {
    rotationMode=BMM_Additive
    rotationSpace=BCS_BoneSpace
    scaleMode=BMM_Additive
    scaleSpace=BCS_BoneSpace
    translationMode=BMM_Additive
    translationSpace=BCS_BoneSpace
    alphaBoolBlend {blendOption=Linear}
    boneToModify {boneName=base_flap_01_r}
    translation {x=-1}
   }
   n23 = Anim Dynamics - Bone: dyn_armband_low!(ComponentPose=@n22, Alpha=1, GravityOverride:Vector="0,0,-980", ExternalForce="0,0,0") -> (Pose) {
    alpha=0.9
    angularDampingOverride=0.98
    angularSpringConstant=1500
    bAngularSpring=true
    bGravityOverrideInSimSpace=true
    bOverrideAngularDamping=true
    bUseGravityOverride=true
    numSolverIterationsPostUpdate=4
    numSolverIterationsPreUpdate=16
    simulationSpace=BoneRelative
    alphaBoolBlend {blendOption=Linear}
    boundBone {boneName=dyn_armband_low}
    componentAppliedLinearAccClamp=<60000 60000 60000>
    componentLinearAccScale=<0.3 0.2 0.2>
    componentLinearVelScale=<0.3 0.2 0.2>
    gravityOverride {z=-980}
    physicsBodyDefinitions [{
      sphereCollisionRadius=0
      boundBone {boneName=dyn_armband_low}
      boxExtents=<15 5 5>
      constraintSetup {angularTargetAxis=AxisZ, twistAxis=AxisY, angularLimitsMax=<25 25 25>, angularLimitsMin=<-25 -25 -25>, angularTarget=<0 0 1>}
      localJointOffset=<0 0 0>
     }]
    relativeSpaceBone {boneName=root}
   }
   n24 = Anim Dynamics - Bone: dyn_armband_hi!(ComponentPose=@n23, Alpha=1, GravityOverride:Vector="0,0,-980", ExternalForce="0,0,0") -> (Pose) {
    alpha=0.9
    angularDampingOverride=0.98
    angularSpringConstant=1500
    bAngularSpring=true
    bGravityOverrideInSimSpace=true
    bOverrideAngularDamping=true
    bUseGravityOverride=true
    numSolverIterationsPostUpdate=4
    numSolverIterationsPreUpdate=16
    simulationSpace=BoneRelative
    alphaBoolBlend {blendOption=Linear}
    boundBone {boneName=dyn_armband_hi}
    componentAppliedLinearAccClamp=<60000 60000 60000>
    componentLinearAccScale=<0.3 0.2 0.2>
    componentLinearVelScale=<0.3 0.2 0.2>
    gravityOverride {z=-980}
    physicsBodyDefinitions [{
      sphereCollisionRadius=0
      boundBone {boneName=dyn_armband_hi}
      boxExtents=<15 5 5>
      constraintSetup {angularTargetAxis=AxisZ, twistAxis=AxisY, angularLimitsMax=<25 25 25>, angularLimitsMin=<-25 -25 -25>, angularTarget=<0 0 1>}
      localJointOffset=<0 0 0>
     }]
    relativeSpaceBone {boneName=root}
   }
   n25 = Anim Dynamics - Bone: dyn_handband_big!(ComponentPose=@n24, Alpha=1, GravityOverride:Vector="0,0,-980", ExternalForce="0,0,0") -> (Pose) {
    alpha=0.9
    angularDampingOverride=0.98
    angularSpringConstant=1500
    bAngularSpring=true
    bGravityOverrideInSimSpace=true
    bOverrideAngularDamping=true
    bUseGravityOverride=true
    numSolverIterationsPostUpdate=4
    numSolverIterationsPreUpdate=16
    simulationSpace=BoneRelative
    alphaBoolBlend {blendOption=Linear}
    boundBone {boneName=dyn_handband_big}
    componentAppliedLinearAccClamp=<60000 60000 60000>
    componentLinearAccScale=<0.3 0.2 0.2>
    componentLinearVelScale=<0.3 0.2 0.2>
    gravityOverride {z=-980}
    physicsBodyDefinitions [{
      sphereCollisionRadius=0
      boundBone {boneName=dyn_handband_big}
      boxExtents=<15 5 5>
      constraintSetup {angularTargetAxis=AxisZ, twistAxis=AxisY, angularLimitsMax=<5 5 5>, angularLimitsMin=<-5 -5 -25>, angularTarget=<0 0 1>}
      localJointOffset=<0 0 0>
     }]
    relativeSpaceBone {boneName=root}
   }
   n26 = Anim Dynamics - Bone: dyn_handband_little!(ComponentPose=@n25, Alpha=1, GravityOverride:Vector="0,0,-980", ExternalForce="0,0,0") -> (Pose) {
    alpha=0.9
    angularDampingOverride=0.98
    angularSpringConstant=1500
    bAngularSpring=true
    bGravityOverrideInSimSpace=true
    bOverrideAngularDamping=true
    bUseGravityOverride=true
    numSolverIterationsPostUpdate=4
    numSolverIterationsPreUpdate=16
    simulationSpace=BoneRelative
    alphaBoolBlend {blendOption=Linear}
    boundBone {boneName=dyn_handband_little}
    componentAppliedLinearAccClamp=<60000 60000 60000>
    componentLinearAccScale=<0.3 0.2 0.2>
    componentLinearVelScale=<0.3 0.2 0.2>
    gravityOverride {z=-980}
    physicsBodyDefinitions [{
      sphereCollisionRadius=0
      boundBone {boneName=dyn_handband_little}
      boxExtents=<15 5 5>
      constraintSetup {angularTargetAxis=AxisZ, twistAxis=AxisY, angularLimitsMax=<5 5 5>, angularLimitsMin=<-5 -5 -25>, angularTarget=<0 0 1>}
      localJointOffset=<0 0 0>
     }]
    relativeSpaceBone {boneName=root}
   }
   Save cached pose 'RB_dynamics_cache'(Pose=Component To Local(ComponentPose=@n26)) {cacheName=RB_dynamics_cache}
  ]}
 EventGraph {nodes [
   Event BlueprintUpdateAnimation()
   > n1 = Sequence!(execute) -> (then_0, then_1, then_2:Exec)
   > n2 = Gravity Override State Function DEPRECATED(CharacterPartAnimInstance=Self-Reference(), GameTimeInSeconds=GetGameTimeInSeconds(), GravityOverrideParams=Get Gravity Override Params())
   > Set GravityOverride(GravityOverride=@n2)
   > Set SkelMesh(SkelMesh=GetPartSkeletalMeshComponent())
   n3 = Get SkelMesh()
   > Set legRotation_L(legRotation_L:double=GetAlphaFromReferenceTransformDelta(SkeletalMeshComponent=@n3, BoneName=pelvis, BaseName=thigh_l, InputRange=(X=0,Y=90), OutputRange=(X=0,Y=0.5), TwistAxis="0,0,1")) -> (Output_Get:double)
   > Set legRotation_R(legRotation_R:double=GetAlphaFromReferenceTransformDelta(SkeletalMeshComponent=@n3, BoneName=pelvis, BaseName=thigh_r, InputRange=(X=0,Y=90), OutputRange=(X=0.5,Y=1), TwistAxis="0,0,1")) -> (Output_Get:double)
   > Set UpperArm_Rot_Alpha(UpperArm_Rot_Alpha:double=GetAlphaFromReferenceTransformDelta(
     SkeletalMeshComponent=Get SkelMesh()
     BoneName=upperarm_r
     BaseName=clavicle_r
     DeltaComponentType=EulerY
     InputRange=(X=-40,Y=0)
     TwistAxis="0,0,1"
    )) -> (Output_Get:double)
   > Set skydiving(skydiving:double="To Float (Boolean)"(InBool=Get bIsParachuteOpened())) -> (Output_Get:double)
   Set calf_l(
    execute=@n1.then_1
    calf_l:double=GetAlphaFromReferenceTransformDelta(SkeletalMeshComponent=Get SkelMesh(), BoneName=thigh_l, BaseName=calf_l, InputRange=(X=-45,Y=-120), TwistAxis="0,0,1")
   ) -> (Output_Get:double)
   > Set calf_r(calf_r:double=GetAlphaFromReferenceTransformDelta(
     SkeletalMeshComponent=Get SkelMesh()
     BoneName=thigh_r
     BaseName=calf_r
     InputRange=(X=-45,Y=-120)
     OutputRange=(X=-0.3,Y=1)
     TwistAxis="0,0,1"
    )) -> (Output_Get:double)
   > n4 = Branch(Condition=OR Boolean(A=Get bIsInFrontEnd(), B=Get bIsDBNO()))
   > Set rotationBigFlap(rotationBigFlap:Rotator="-105,5,0") -> (Output_Get:Rotator)
   > Set rotationLittleFlap(rotationLittleFlap:Rotator="-65,0,0") -> (Output_Get:Rotator)
   Set rotationBigFlap(execute=@n4.else, rotationBigFlap="-80,0,0")
   > Set rotationLittleFlap(rotationLittleFlap="-45,0,0")
  ]}
]
variables [
 calf_l {type=number}
 calf_r {type=number}
 Gravity Override Params {type=GravityOverrideParamsStruct}
 GravityOverride {type=Vector}
 isCrounchingAlpha {type=number}
 legRotation_L {type=number}
 legRotation_R {type=number}
 rotationBigFlap {type=Rotator}
 rotationLittleFlap {type=Rotator}
 SkelMesh {type=SkeletalMeshComponent}
 skydiving {type=number}
 UpperArm_Rot_Alpha {type=number}
]
